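# Physics asset for the CrumpleFold backpack cosmetic, giving capsule bodies and angular joint limits to its pack, dampers, straps, handle and head extra joint.
# The middle damper branches into left and right dampers, each carrying a strap, with twist locked along the chain.
# Small offset bodies anchor the middle damper, handle and pocket straps.
# Only the pocket strap anchors, fully locked, and the tightly limited head joint constrain to the pack.
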PhysicsAsset /BRCosmetics/Accessories/FORT_Backpacks/Backpack_M_MED_CrumpleFold/Meshes/M_MED_CrumpleFold_Pack_Physics
bodies {
 Offest_pocket_strap_left_grp [{shapeType=Capsule, length=0.505, radius=0.505, shapeName=Offest_pocket_strap_left_grp_capsule, rotation=<rot 0 -90 0>}]
 Offset_backpack_handle_grp [{shapeType=Capsule, length=0.505, radius=0.505, shapeName=Offset_backpack_handle_grp_capsule, rotation=<rot 0 -90 0>}]
 Offset_damper_bp_mid_grp [{shapeType=Capsule, length=0.505, radius=0.505, shapeName=Offset_damper_bp_mid_grp_capsule, rotation=<rot 0 -90 0>}]
 Offset_pocket_strap_right_grp [{shapeType=Capsule, length=0.505, radius=0.505, shapeName=Offset_pocket_strap_right_grp_capsule, rotation=<rot 0 -90 0>}]
 dyn_backpack_handle_jnt [{
   shapeType=Capsule
   length=3.5153
   radius=1.4547
   shapeName=dyn_backpack_handle_jnt_capsule
   center=<0 0 2>
   rotation=<rot 0 -92.8663 -81.2734>
  }]
 dyn_damper_bp_left_jnt [{
   shapeType=Capsule
   length=5.1913
   radius=2.8087
   shapeName=dyn_damper_bp_left_jnt_capsule
   center=<-3.1348 0 -0.7264>
   rotation=<rot -87.3826 100.2162 -109.8924>
  }]
 dyn_damper_bp_mid_jnt [{
   shapeType=Capsule
   length=8.6352
   radius=2.9467
   shapeName=dyn_damper_bp_mid_jnt_capsule
   center=<-4.9703 0 -1.8798>
   rotation=<rot -65.7371 0.0001 8.4983>
  }]
 dyn_damper_bp_right_jnt [{
   shapeType=Capsule
   length=5.4171
   radius=2.5829
   shapeName=dyn_damper_bp_right_jnt_capsule
   center=<-3.7611 -0.6504 -0.3595>
   rotation=<rot -84.5392 0.0004 14.9996>
  }]
 dyn_pocket_strap_left_jnt [{
   shapeType=Capsule
   length=2.9407
   radius=0.505
   shapeName=dyn_pocket_strap_left_jnt_capsule
   center=<0.0931 0.1739 -1.3222>
   rotation=<rot -3.321 -90 3.3604>
  }]
 dyn_pocket_strap_right_jnt [{
   shapeType=Capsule
   length=3.1143
   radius=0.505
   shapeName=dyn_pocket_strap_right_jnt_capsule
   center=<0.0478 0.3597 -1.3384>
   rotation=<rot -8.2735 -90 -0.5009>
  }]
 dyn_strap_left_jnt [{
   shapeType=Capsule
   length=1.5247
   radius=0.8405
   shapeName=dyn_strap_left_jnt_capsule
   center=<1.2671 0.1594 -0.4568>
   rotation=<rot 0 -80.2342 -68.0784>
  }]
 dyn_strap_right_jnt [{
   shapeType=Capsule
   length=1.493
   radius=0.8164
   shapeName=dyn_strap_right_jnt_capsule
   center=<1.5015 0.0683 -0.1569>
   rotation=<rot 0 -90.4126 -78.1853>
  }]
 head_backP_extra_jnt [{shapeType=Capsule, length=2.5164, radius=6.3853, shapeName=head_backP_extra_jnt_capsule, rotation=<rot 4.9888 117.5007 -3.8564>}]
 pack [
  {
   shapeType=Capsule
   length=6.0828
   radius=6.5098
   shapeName=pack_capsule
   center=<-2.3649 -10.1964 -5.1568>
   rotation=<rot -2.3896 87.2014 70.9675>
  }
  {shapeType=Capsule, length=18.0557, radius=5.3756, shapeName=pack_capsule, center=<-1.1796 -5.91 -11.3979>, rotation=<rot 90 -90 -91.2149>}
 ]
}
constraints [
 {
  bone1Name=dyn_damper_bp_left_jnt
  bone2Name=dyn_damper_bp_mid_jnt
  swing1LimitDegrees=5
  swing2LimitDegrees=35
  twistLimitDegrees=45
  twistMotion=Locked
 }
 {
  bone1Name=dyn_damper_bp_right_jnt
  bone2Name=dyn_damper_bp_mid_jnt
  swing1LimitDegrees=5
  swing2LimitDegrees=35
  twistLimitDegrees=45
  twistMotion=Locked
 }
 {
  bone1Name=dyn_damper_bp_mid_jnt
  bone2Name=Offset_damper_bp_mid_grp
  swing1LimitDegrees=5
  swing2LimitDegrees=35
  twistLimitDegrees=45
  twistMotion=Locked
 }
 {
  bone1Name=dyn_strap_right_jnt
  bone2Name=dyn_damper_bp_right_jnt
  swing1LimitDegrees=25
  swing2LimitDegrees=45
  twistLimitDegrees=45
  twistMotion=Locked
 }
 {
  bone1Name=dyn_strap_left_jnt
  bone2Name=dyn_damper_bp_left_jnt
  swing1LimitDegrees=25
  swing2LimitDegrees=45
  twistLimitDegrees=45
  twistMotion=Locked
 }
 {
  bone1Name=dyn_pocket_strap_left_jnt
  bone2Name=Offest_pocket_strap_left_grp
  swing1LimitDegrees=2
  swing2LimitDegrees=10
  twistLimitDegrees=25
 }
 {
  bone1Name=dyn_pocket_strap_right_jnt
  bone2Name=Offset_pocket_strap_right_grp
  swing1LimitDegrees=2
  swing2LimitDegrees=10
  twistLimitDegrees=25
 }
 {
  bone1Name=dyn_backpack_handle_jnt
  bone2Name=Offset_backpack_handle_grp
  swing1LimitDegrees=15
  swing1Motion=Locked
  swing2LimitDegrees=5
  twistLimitDegrees=15
 }
 {
  bone1Name=Offest_pocket_strap_left_grp
  bone2Name=pack
  swing1LimitDegrees=45
  swing1Motion=Locked
  swing2LimitDegrees=45
  swing2Motion=Locked
  twistLimitDegrees=45
  twistMotion=Locked
 }
 {
  bone1Name=Offset_pocket_strap_right_grp
  bone2Name=pack
  swing1LimitDegrees=45
  swing1Motion=Locked
  swing2LimitDegrees=45
  swing2Motion=Locked
  twistLimitDegrees=45
  twistMotion=Locked
 }
 {bone1Name=head_backP_extra_jnt, bone2Name=pack, swing1LimitDegrees=2, swing2LimitDegrees=2, twistLimitDegrees=1}
]
