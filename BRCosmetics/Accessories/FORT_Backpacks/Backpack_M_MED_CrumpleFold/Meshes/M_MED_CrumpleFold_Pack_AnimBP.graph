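# Post-process style animation Blueprint for the CrumpleFold backpack cosmetic, animating its head, teeth and fabric bones.
# Fixed bone offsets adjust the input pose, which then receives one of two additive idle loops, chosen by front-end state.
# A RigidBody simulation in pack space adds secondary motion, followed by an upper-teeth nudge at LOD 1 or 2.
# The computed gravity override, constraint nodes and second RigidBody node are not wired into the output.
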AnimBlueprint /BRCosmetics/Accessories/FORT_Backpacks/Backpack_M_MED_CrumpleFold/Meshes/M_MED_CrumpleFold_Pack_AnimBP
target_skeleton=&/M_MED_CrumpleFold_Pack_Skeleton
graphs [
 AnimGraph {nodes [
   n1 = "Constraint - dyn_damper_bp_left_jnt (P)"(Alpha=0.1, ConstraintWeights_0=1) {
    alphaBoolBlend {blendOption=Linear}
    boneToModify {boneName=dyn_damper_bp_left_jnt}
    constraintSetup [{targetBone {boneName=dyn_damper_bp_mid_jnt}}]
    constraintWeights [1]
   }
   "Constraint - dyn_damper_bp_right_jnt (P)"(ComponentPose=@n1, Alpha=0.1, ConstraintWeights_0=1) {
    alphaBoolBlend {blendOption=Linear}
    boneToModify {boneName=dyn_damper_bp_right_jnt}
    constraintSetup [{targetBone {boneName=dyn_damper_bp_mid_jnt}}]
    constraintWeights [1]
   }
   RigidBody(Alpha=0.2) {
    bOverrideWorldGravity=true
    overridePhysicsAsset=&/M_MED_CrumpleFold_Pack_Physics
    simulationSpace=WorldSpace
    alphaBoolBlend {blendOption=Linear}
    baseBoneRef {boneName=pack}
    componentAppliedLinearAccClamp=<1000 1000 1000>
    componentLinearAccScale=<0.1 0.25 0.8>
    componentLinearVelScale=<0.1 0.25 0.5>
    overrideWorldGravity {z=-980}
    simSpaceSettings {maxAngularAcceleration=1350, maxAngularVelocity=100, maxLinearAcceleration=250, maxLinearVelocity=1, worldAlpha=1}
   }
   AnimGraphNode_StateMachine_0 = Locomotion() {is=title AnimGraphNode_StateMachine}
   n2 = Apply Additive(Base=Use cached pose 'Input'(), Additive=@AnimGraphNode_StateMachine_0) {alphaBoolBlend {blendOption=Linear}}
   AnimGraphNode_StateMachine_1 = FE_Locomotion() {is=title AnimGraphNode_StateMachine}
   n3 = Apply Additive(Base=Use cached pose 'Input'(), Additive=@AnimGraphNode_StateMachine_1) {alphaBoolBlend {blendOption=Linear}}
   n4 = Get bIsInFrontEnd()
   n5 = RigidBody!(
    ComponentPose=Local To Component(LocalPose=Blend Poses by bool(BlendPose_0=@n3, BlendPose_1=@n2, bActiveValue=@n4) {blendType=Linear})
    Alpha
    OverrideWorldGravity:Vector="0,0,-980"
    ExternalForce
   ) -> (Pose) {
    bClampLinearTranslationLimitToRefPose=true
    bOverrideWorldGravity=true
    overlapChannel=ECC_WorldDynamic
    overridePhysicsAsset=&M_MED_CrumpleFold_Pack_Physics
    simulationSpace=BaseBoneSpace
    alphaBoolBlend {blendOption=Linear}
    baseBoneRef {boneName=pack}
    overrideWorldGravity {z=-980}
    simSpaceSettings {
     maxAngularAcceleration=40
     maxAngularVelocity=8
     maxLinearAcceleration=3500
     maxLinearVelocity=700
     worldAlpha=0.4
     externalAngularVelocity=<1 1 1>
     externalLinearVelocity {z=1}
    }
   }
   n6 = "Transform (Modify) Bone - Bone: TeethUp_backP_extra_jnt"!(
    ComponentPose=@n5
    bAlphaBoolEnabled:bool=Get Is lod 1 or 2() -> (Is lod 1 or 2:bool)
    Translation="0,0,-0.2"
    Rotation="0,0,0"
    Scale="1,1,1"
   ) -> (Pose) {
    alphaInputType=Bool
    rotationMode=BMM_Additive
    rotationSpace=BCS_BoneSpace
    scaleMode=BMM_Additive
    translationMode=BMM_Additive
    translationSpace=BCS_BoneSpace
    alphaBoolBlend {blendOption=Linear}
    boneToModify {boneName=TeethUp_backP_extra_jnt}
    translation {z=-0.2}
   }
   Output Pose(Result=Component To Local(ComponentPose=@n6)) {layerGroup=None}
   n7 = "Transform (Modify) Bone - Bone: head_backP_extra_jnt"(ComponentPose=Local To Component(LocalPose=Input Pose()), Alpha=1, Translation="0,0,0.75", Rotation="25,15,-10", Scale="1,1,1") {
    rotationMode=BMM_Additive
    rotationSpace=BCS_BoneSpace
    translationMode=BMM_Additive
    translationSpace=BCS_BoneSpace
    alphaBoolBlend {blendOption=Linear}
    boneToModify {boneName=head_backP_extra_jnt}
    rotation=<rot 25 15 -10>
    translation {z=0.75}
   }
   n8 = "Transform (Modify) Bone - Bone: Offset_damper_bp_mid_grp"(ComponentPose=@n7, Alpha=1, Translation="0,0,1.5", Rotation="0,0,0", Scale="1,1,1") {
    rotationMode=BMM_Additive
    rotationSpace=BCS_BoneSpace
    translationMode=BMM_Additive
    translationSpace=BCS_BoneSpace
    alphaBoolBlend {blendOption=Linear}
    boneToModify {boneName=Offset_damper_bp_mid_grp}
    translation {z=1.5}
   }
   n9 = "Transform (Modify) Bone - Bone: bag_fabricControl_back_jnt"(ComponentPose=@n8, Alpha=1, Translation="0,2,0", Rotation="0,0,0", Scale="1,1,1") {
    rotationMode=BMM_Additive
    rotationSpace=BCS_BoneSpace
    translationMode=BMM_Additive
    translationSpace=BCS_BoneSpace
    alphaBoolBlend {blendOption=Linear}
    boneToModify {boneName=bag_fabricControl_back_jnt}
    translation {y=2}
   }
   n10 = "Transform (Modify) Bone - Bone: bag_fabricControl_mid_jnt"(ComponentPose=@n9, Alpha=1, Translation="2,0,3", Rotation="0,0,0", Scale="1,1,1") {
    rotationMode=BMM_Replace
    rotationSpace=BCS_BoneSpace
    translationMode=BMM_Additive
    translationSpace=BCS_BoneSpace
    alphaBoolBlend {blendOption=Linear}
    boneToModify {boneName=bag_fabricControl_mid_jnt}
    translation {x=2, z=3}
   }
   n11 = "Transform (Modify) Bone - Bone: TeethUp_backP_extra_jnt"!(ComponentPose=@n10, Alpha:float=1, Translation="0.05,0,-0.2", Rotation="0,0,0", Scale="0.95,1,1") -> (Pose) {
    rotationMode=BMM_Additive
    rotationSpace=BCS_BoneSpace
    scaleMode=BMM_Additive
    translationMode=BMM_Additive
    translationSpace=BCS_BoneSpace
    alphaBoolBlend {blendOption=Linear}
    boneToModify {boneName=TeethUp_backP_extra_jnt}
    scale {x=0.95}
    translation {x=0.05, z=-0.2}
   }
   n12 = "Transform (Modify) Bone - Bone: TeethLow_backP_extra_jnt"(ComponentPose=@n11, Alpha=1, Translation="0,0,0", Rotation="0,0,0", Scale="0.95,1,1") {
    rotationMode=BMM_Additive
    rotationSpace=BCS_BoneSpace
    scaleMode=BMM_Additive
    translationMode=BMM_Additive
    translationSpace=BCS_BoneSpace
    alphaBoolBlend {blendOption=Linear}
    boneToModify {boneName=TeethLow_backP_extra_jnt}
    scale {x=0.95}
   }
   Save cached pose 'Input'(Pose=Component To Local(ComponentPose=@n12)) {cacheName=Input}
  ]}
 Locomotion {entry=Idle, path=AnimGraph.AnimGraphNode_StateMachine_0.Locomotion, ids=id, nodes [AnimStateNode_0 = Idle()]}
 Idle {
  path=AnimGraph.AnimGraphNode_StateMachine_0.Locomotion.AnimStateNode_0.Idle
  nodes [Output Animation Pose(Result="Sequence Player (additive) 'CrumpleFold_Idle_headMotion_WEP'"() {sequence=/BRCosmetics/Accessories/FORT_Backpacks/Backpack_M_MED_CrumpleFold/Animations/CrumpleFold_Idle_headMotion_WEP})]
 }
 FE_Locomotion {entry=Idle, path=AnimGraph.AnimGraphNode_StateMachine_1.FE_Locomotion, ids=id, nodes [AnimStateNode_0 = Idle()]}
 Idle {
  path=AnimGraph.AnimGraphNode_StateMachine_1.FE_Locomotion.AnimStateNode_0.Idle
  nodes [Output Animation Pose(Result="Sequence Player (additive) 'CrumpleFold_Idle_PACK'"() {sequence=/BRCosmetics/Accessories/FORT_Backpacks/Backpack_M_MED_CrumpleFold/Animations/CrumpleFold_Idle_PACK})]
 }
 EventGraph {nodes [
   Event BlueprintUpdateAnimation()
   > n1 = Sequence()
   > n2 = Gravity Override State Function DEPRECATED(CharacterPartAnimInstance=Self-Reference(), GameTimeInSeconds=GetGameTimeInSeconds(), GravityOverrideParams=Get Gravity Override Params())
   > Set Gravity Override(Gravity Override=@n2)
   n3 = Get CurrentLOD()
   n4 = Get CurrentLOD()
   Set Is lod 1 or 2(execute=@n1.then_1, Is lod 1 or 2:bool=OR Boolean(A="Equal (Integer)"(A=@n4, B=2), B="Equal (Integer)"(A=@n3, B=1))) -> (Output_Get:bool)
   TryGetPawnOwner()
  ]}
]
variables [
 Gravity Override {type=Vector}
 Gravity Override Params {type=GravityOverrideParamsStruct}
 Is lod 1 or 2 {type=boolean}
]
